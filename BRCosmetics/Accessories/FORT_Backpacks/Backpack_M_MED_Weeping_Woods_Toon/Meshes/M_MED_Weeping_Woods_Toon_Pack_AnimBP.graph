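# Animation Blueprint for the Weeping Woods Toon back bling, switching its bee nest between idle and moving animations.
# Each update caches the pawn and player anim instance, setting IsKeepBeesClose when the player moves, melees or emotes.
# That flag blends the two sync-grouped sequences, feeding an Anim Dynamics spring on the nest_dyn bone.
# The WeepingWoodsAnimTracking curve, read from this instance, sets the dynamics alpha.
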AnimBlueprint /BRCosmetics/Accessories/FORT_Backpacks/Backpack_M_MED_Weeping_Woods_Toon/Meshes/M_MED_Weeping_Woods_Toon_Pack_AnimBP
target_skeleton=&/M_MED_Weeping_Woods_Toon_Pack_Skeleton
graphs [
 AnimGraph {nodes [
   Local Space Ref Pose()
   n1 = Sequence Player 'Weeping_Woods_Toon_Idle'() {
    groupName=Sync
    method=SyncGroup
    sequence=/BRCosmetics/Accessories/FORT_Backpacks/Backpack_M_MED_Weeping_Woods_Toon/Animations/Weeping_Woods_Toon_Idle
   }
   n2 = Sequence Player 'Weeping_Woods_Toon_Moving'() {
    groupName=Sync
    method=SyncGroup
    sequence=/BRCosmetics/Accessories/FORT_Backpacks/Backpack_M_MED_Weeping_Woods_Toon/Animations/Weeping_Woods_Toon_Moving
   }
   n3 = Get IsKeepBeesClose() -> (IsKeepBeesClose:bool)
   n4 = Anim Dynamics - Bone: nest_dyn(
    ComponentPose=Local To Component(LocalPose=Blend Poses by bool(BlendPose_0=@n2, BlendPose_1=@n1, BlendTime_0=0.2, BlendTime_1=0.2, bActiveValue=@n3) {blendType=Linear})
    Alpha=Get AnimTracking() -> (AnimTracking:double)
    ExternalForce="0,0,0"
   ) {
    bAngularSpring=true
    bLinearSpring=true
    bOverrideAngularBias=true
    bOverrideAngularDamping=true
    bUseGravityOverride=true
    linearSpringConstant=15
    simulationSpace=Actor
    alphaBoolBlend {blendOption=Linear}
    boundBone {boneName=nest_dyn}
    gravityOverride=<1 1 1>
    physicsBodyDefinitions [{
      sphereCollisionRadius=0
      boundBone {boneName=nest_dyn}
      boxExtents=<100 100 100>
      constraintSetup {
       angularTargetAxis=AxisY
       coneAngle=60
       twistAxis=AxisY
       angularLimitsMax=<20 0 20>
       angularLimitsMin=<-20 0 -30>
       angularTarget=<0 1 0>
       linearAxesMax=<20 0 20>
       linearAxesMin=<-20 -10 -20>
      }
      localJointOffset=<0 -0.2 0>
     }]
   }
   Output Pose(Result=Component To Local(ComponentPose=@n4)) {layerGroup=None}
  ]}
 EventGraph {nodes [
   Event BlueprintUpdateAnimation()
   > n1 = Branch(Condition=IsNullItem!(ItemObject:Object Reference=Get FortPlayerPawn() -> (FortPlayerPawn:Fort Player Pawn Object Reference)) -> (Is Null Item:bool))
   n2 = TryGetPawnOwner()
   > Is Valid(InputObject=@n2)
   > n3 = Cast To FortPlayerPawn(Object=@n2)
   > n4 = Set FortPlayerPawn(FortPlayerPawn:Fort Player Pawn Object Reference=@n3) -> (Output_Get:Fort Player Pawn Object Reference)
   n5 = Branch(
    execute=[@n1.else, @n4.then]
    Condition=IsNullItem(ItemObject=Get PlayerAnimBlueprint() -> (PlayerAnimBlueprint:Fortnite M Avg Player Anim Blueprint Object Reference))
   )
   n6 = GetAnimInstance(self=Get Mesh(self=Get FortPlayerPawn()))
   > Is Valid(InputObject=@n6)
   > n7 = Cast To Fortnite_M_Avg_Player_AnimBlueprint!(execute, Object=@n6) -> (then, CastFailed, AsFortnite M Avg Player Anim Blueprint:Fortnite M Avg Player Anim Blueprint Object Reference) {is=title K2Node_DynamicCast}
   > n8 = Set PlayerAnimBlueprint(PlayerAnimBlueprint:Fortnite M Avg Player Anim Blueprint Object Reference=@n7) -> (Output_Get:Fortnite M Avg Player Anim Blueprint Object Reference)
   Set AnimTracking(execute=[@n5.else, @n8.then], AnimTracking:double=GetCurveValue(CurveName=WeepingWoodsAnimTracking)) -> (Output_Get:double)
   n9 = Get PlayerAnimBlueprint()
   > Set IsKeepBeesClose(IsKeepBeesClose:bool=OR Boolean!(
     A=Get bIsMoving2D!(self:Fort Player Anim Instance Object Reference=@n9) -> (bIsMoving2D)
     B=Get bIsPlayingMeleeAnim!(self:Fort Player Anim Instance Object Reference=@n9) -> (bIsPlayingMeleeAnim)
     C:bool=Get bIsPlayingEmote!(self:Fort Player Anim Instance Object Reference=@n9) -> (bIsPlayingEmote)
    ) -> (ReturnValue)) -> (Output_Get:bool)
   OR Boolean!(
    A=Get bIsMoving2D!() -> (bIsMoving2D)
    B=Get bIsPlayingMeleeAnim!() -> (bIsPlayingMeleeAnim)
    C=Get bIsPlayingEmote!() -> (bIsPlayingEmote)
    D:bool
   ) -> (ReturnValue)
  ]}
]
variables [
 AnimTracking {type=number}
 FortPlayerPawn {type=FortPlayerPawn}
 IsKeepBeesClose {type=boolean}
 PlayerAnimBlueprint {type=Fortnite_M_Avg_Player_AnimBlueprint_C, default=null}
]
